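# ParticleDOF is a library material function that fakes depth of field on particle sprites, for use with CircleDOF.
# It outputs a World Position Offset that grows the sprite along mirrored, centered UVs by the circle-of-confusion radius plus MinPixelRadius, converted from pixels to world units.
# Opacity is scaled by the squared ratio of real to enlarged pixel radius to preserve brightness.
# The MinPixelRadius parameter also reduces aliasing.
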MaterialFunction /Engine/Functions/Engine_MaterialFunctions03/Particles/ParticleDOF
graph {
 properties {bExposeToLibrary=true, description=Use with with CircleDOF}
 nodes [
  n1 = Multiply(
   A=Subtract(A=TextureCoordinate()) {constB=0.5}
   B=Constant2Vector() {desc=mirror it that the texture works and no double sided is needed, g=-2, r=2}
  ) {constB=-2}
  n2 = ScalarParameter() {defaultValue=1, desc=to reduce aliasing, parameterName=MinPixelRadius}
  n3 = Abs(None=Multiply(A=DepthOfFieldFunction!(Depth) {functionValue=TDOF_CircleOfConfusionRadius}) {constA=4}) {desc="Radius DOFFunction=absolute additive radius"}
  FunctionOutput(None=Multiply(
    A=MaterialFunctionCall!(Pixel=@n1, Depth) {materialFunction=/Engine/Functions/Engine_MaterialFunctions01/Coordinates/PixelInWorldUnits}.PixelInWorldUnits
    B=Add(A=@n2, B=@n3) {constB=5, desc="Radius (DOF,Min)"}
   ) {desc=OUT WorldPositionOffset}) {outputName=WorldPositionOffset}
  n4 = MaterialFunctionCall!(WorldUnits=ParticleRadius() {desc=in WorldSpace}, Depth) {materialFunction=/Engine/Functions/Engine_MaterialFunctions01/Coordinates/WorldUnitsInPixel}
  n5 = Multiply(A=@n4.PIxels, B=@n4.PIxels)
  n6 = Add(A=Add(A=@n2, B=@n4.PIxels), B=@n3) {desc="Radius (DOF,Min,Real) for Brightness "}
  n7 = Multiply(A=@n6, B=@n6)
  FunctionOutput(None=Multiply(
    A=Divide(A=@n5, B=@n7) {desc="brightness correction factor (using RenderRadius)"}
    B=FunctionInput() {inputName=Opacity, inputType=FunctionInput_Scalar, previewValue=<1 0 0 1>}
   ) {desc=OUT Opacity}) {outputName=Opacity}
 ]
}
