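# Translucent, unlit particle sprite material for sparkles, apparently for the disco ball capture effect, with responsive AA enabled.
# Color blends between two blue parameters by particle random, replaced by a constant on ES2 and ES3.1.
# Two summed sines add twinkle, with particle green and random scaling speed and particle red scaling brightness.
# A channel-masked sparkle texture times particle alpha feeds ParticleDOF for opacity.
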Material /Clyde/Athena/Effects/Materials/Disco/M_Disco_Ball_Capture_Sparkle
graph {
 properties {
  bEnableResponsiveAA=true
  bUsedWithParticleSprites=true
  blendMode=BLEND_Translucent
  pixelDepthOffsetMode=PDOM_Legacy
  shadingModel=MSM_Unlit
  usedShadingModels=Unlit
 }
 nodes [
  n1 = ParticleColor()
  n2 = TextureSampleParameter2D() {parameterName=Opacity Texture, texture=/Game/Effects/Fort_Effects/Textures/Sparkle/T_Sparkle_01}
  MaterialFunctionCall!(Opacity=Multiply(A=@n1.A, B=StaticComponentMaskParameter(None=@n2.RGB) {defaultG=true, parameterName=Opacity Component Mask})) {materialFunction=/Engine/Functions/Engine_MaterialFunctions03/Particles/ParticleDOF}
  n3 = VectorParameter() {parameterName=Color A, defaultValue=<rgba 0.116 0.3229 1 0>}
  n4 = VectorParameter() {parameterName=Color B, defaultValue=<rgba 0.0746 0.8197 1 0>}
  n5 = ParticleRandom()
  n6 = Constant()
  n7 = FeatureLevelSwitch(Default=@n5, ES2=@n6, ES3_1=@n6, SM4_REMOVED=@n5)
  n8 = Multiply(A=Time(), B=Multiply(A=@n5, B=ParticleColor().G))
  n9 = Add(A=Sine(None=@n8) {period=0.18}, B=Sine(None=@n8) {period=0.42})
  Multiply(A=@n1.R, B=Multiply(A=LinearInterpolate(A=@n3.RGB, B=@n4.RGB, Alpha=@n7), B=Add(A=@n9) {constB=2}))
 ]
}
