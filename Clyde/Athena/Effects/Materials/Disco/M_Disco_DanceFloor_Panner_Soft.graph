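# Translucent unlit mesh-particle material for a disco dance floor effect, using a tiled dance floor texture on the second UV channel.
# The texture multiplied by particle color forms the color term.
# A second term multiplies the squared texture by particle alpha and a SphereMask with Radius and Hardness parameters.
# Zeroed U tiling makes that mask a band along V, positioned by the "Panning" dynamic parameter.
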Material /Clyde/Athena/Effects/Materials/Disco/M_Disco_DanceFloor_Panner_Soft
graph {
 properties {bUsedWithMeshParticles=true, blendMode=BLEND_Translucent, pixelDepthOffsetMode=PDOM_Legacy, shadingModel=MSM_Unlit, usedShadingModels=Unlit}
 nodes [
  n1 = TextureSample(UVs=TextureCoordinate() {coordinateIndex=1, uTiling=3, vTiling=3}) {texture=/Game/Environments/Sets/Concert/Textures/T_DanceFloor_D}
  n2 = ParticleColor()
  Multiply(A=@n1.RGB, B=@n2.RGB)
  n3 = TextureCoordinate() {uTiling=0}
  n4 = AppendVector(A=Constant(), B=DynamicParameter() {paramNames [Panning,Param2,Param3,Param4]}.Panning)
  n5 = ScalarParameter() {defaultValue=0.15, parameterName=Radius}
  n6 = ScalarParameter() {defaultValue=-25, parameterName=Hardness}
  Multiply(A=Multiply(A=@n1.RGB, B=@n1.RGB), B=Multiply(A=@n2.A, B=SphereMask(A=@n3, B=@n4, Radius=@n5, Hardness=@n6)))
  Power(Base=@n1.RGB)
 ]
}
