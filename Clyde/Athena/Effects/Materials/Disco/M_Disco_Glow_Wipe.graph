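# Unlit translucent mesh-particle material for a disco glow effect with an animated vertical wipe.
# The GlowPan dynamic parameter slides a soft gradient along the UVs, masked by a vertical fade, particle alpha and near-camera fade for opacity.
# A static switch colors it with either three panning, offset smoke samples packed into RGB for chromatic separation, or a smoke-phased sine pulse between DarkColor and white.
# The wipe gradient brightens that color, which particle color tints.
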Material /Clyde/Athena/Effects/Materials/Disco/M_Disco_Glow_Wipe
graph {
 properties {bUsedWithMeshParticles=true, blendMode=BLEND_Translucent, pixelDepthOffsetMode=PDOM_Legacy, shadingModel=MSM_Unlit, usedShadingModels=Unlit}
 nodes [
  n1 = ParticleColor()
  n2 = TextureCoordinate() {vTiling=6}
  n3 = TextureCoordinate()
  n4 = TextureSample(UVs=LinearInterpolate(
    A=Add(A=@n2) {constB=-5.5}
    B=Add(A=@n3) {constB=0.7}
    Alpha=DynamicParameter() {defaultValue=<rgba 1 1 0 1>, paramNames [FresnelExponent, Refraction, GlowPan, Param4]}.GlowPan
   )) {texture=/Clyde/Effects/Fort_Effects/Textures/Masks/T_Vert_Gradient_Soft}
  n5 = Add(A=@n4.R) {constB=0.1}
  n6 = TextureCoordinate()
  n7 = TextureSample(UVs=OneMinus(None=ComponentMask(None=@n6) {g=true})) {texture=/Game/Effects/Fort_Effects/Textures/Masks/T_Vert_Gradient_Softer}
  Multiply(
   A=Multiply(A=@n1.A, B=Multiply(A=@n5, B=@n7.R))
   B=MaterialFunctionCall(Distance=ScalarParameter() {defaultValue=500, parameterName=NearCamFade}) {materialFunction=&Near_Camera_Fade}.Result
  )
  n8 = VectorParameter() {parameterName=UVScale, defaultValue=<rgba 2 1 0 1>}
  n9 = Multiply(
   A=Multiply(A=ComponentMask(None=@n8.RGB) {g=true, r=true}, B=TextureCoordinate())
   B=ScalarParameter() {defaultValue=1, parameterName=SquareTile}
  )
  n10 = ComponentMask(None=VectorParameter() {parameterName=ChromaOffset, defaultValue=<rgba 0.05 -0.0125 0 0>}.RGB) {g=true, r=true}
  n11 = Panner(Coordinate=Add(A=@n9, B=@n10)) {speedY=0.125}
  n12 = Panner(Coordinate=@n9) {speedY=0.125}
  n13 = Add(A=@n9, B=Multiply(A=@n10) {constB=-1})
  n14 = AppendVector(
   A=AppendVector(
    A=TextureSample(UVs=@n11) {texture=/Game/Effects/Fort_Effects/Textures/Tile/T_Inky_Smoke_Tile}.R
    B=TextureSample(UVs=@n12) {texture=&T_Inky_Smoke_Tile}.R
   )
   B=TextureSample(UVs=Panner(Coordinate=@n13) {speedY=0.125}) {texture=&T_Inky_Smoke_Tile}.R
  )
  n15 = Add(A=TextureSample(UVs=@n9) {texture=&T_Inky_Smoke_Tile}.R, B=Multiply(A=Time()) {constB=1.5})
  n16 = LinearInterpolate(
   A=VectorParameter() {parameterName=DarkColor, defaultValue=<rgba 0.1178 0.6628 1 1>}.RGB
   Alpha=Abs(None=MaterialFunctionCall!(Sine Phase=@n15, Value 1, Value 2) {materialFunction=/Engine/Functions/Engine_MaterialFunctions02/Utility/Sine_Remapped}.Remapped Sine)
  ) {constA=1}
  n17 = Power(Base=@n4.R) {constExponent=1}
  Multiply(A=@n1.RGB, B=Add(A=StaticSwitchParameter(True=@n14, False=@n16) {parameterName=Use Chromatic}, B=Multiply(A=@n17) {constB=0.25}) {constA=1})
 ]
}
